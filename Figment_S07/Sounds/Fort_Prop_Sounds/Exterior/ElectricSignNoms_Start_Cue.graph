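# Sound cue for the start sound of an electric "Noms" sign, an exterior prop sound in Figment_S07.
# A single wave player feeds the root directly with one start wave, with no randomization, mixing or modulation.
# The cue borrows the gadget-activate attenuation and the gadgets sound class, so it falls off over distance and mixes as a gadget sound.
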SoundCue /Figment_S07/Sounds/Fort_Prop_Sounds/Exterior/ElectricSignNoms_Start_Cue
root=WavePlayer_0
nodes [{id=WavePlayer_0, properties {soundWaveAssetPtr=&/ElectricSignNoms_Start_01}}]
properties {
 attenuationSettings=&Fort_GadgetActivate_Att
 maxDistance=4000
 soundClassObject=&Fort_Gadgets
}
